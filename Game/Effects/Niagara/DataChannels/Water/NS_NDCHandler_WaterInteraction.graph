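# Niagara data channel handler that renders water impact splashes, apparently for weapon hits.
# Emitters spawn from a shared NDC reader, reading impact point, normal, water color and ammo type to drive sprite drops (one GPU-simulated), a short-lived debris sprite and foam meshes.
# Foam scales by ammo type, stays constrained to the water plane and passes water body, depth and wave data to its material.
# Several alternate drop, foam and mist emitters are disabled.
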NiagaraSystem /Game/Effects/Niagara/DataChannels/Water/NS_NDCHandler_WaterInteraction
dependencies {
 usedDataInterfaces [
  &NiagaraDataInterfaceArrayFloat
  &NiagaraDataInterfaceCurve
  &NiagaraDataInterfaceSpriteRendererInfo
  /Script/Niagara.NiagaraDataInterfaceDataChannelRead
 ]
 usedDynamicInputs []
 usedModules [
  &self:SystemSpawnScript.SystemScriptSource.SystemScriptGraph.NiagaraNodeAssignment_1.SetVariables_#1
  &self:SystemSpawnScript.SystemScriptSource.SystemScriptGraph.NiagaraNodeAssignment_2.SetVariables_#2
  &SystemState
  &self:SystemSpawnScript.SystemScriptSource.SystemScriptGraph.NiagaraNodeAssignment_0.SetVariables_#3
  /Niagara/Modules/System/CompleteIfUnused
 ]
}
emitters {
 Drops_Tall {
  bEnabled=false
  emitterUpdateScript [
   &EmitterState()
   /Game/Effects/Niagara/DataChannels/Impacts/Spawn_NDCFortImpacts(Max Spawn Count=5, Min Spawn Count=3, Reader:NiagaraDataInterfaceDataChannelRead=$System.NDCReader)
  ]
  particleSpawnScript [
   &/InitializeParticleFromNDC_FortWaterInteraction()
   /Game/Effects/Niagara/Modules/Color/HSV_Offset(Input Color:LinearColor=&MakeCustomLinearColorFromBool(Bool=$System.DoDebugColor, True Color=$System.DebugWaterColor, False Color=$Output.InitializeParticleFromNDC_FortWaterInteraction.WaterColor))
   /SetVariables(Particles.AmmoTypeScale:NiagaraFloat=&Multiply_Float(
     A=&SelectFloatFromArray(
      Direct Array Index=$Output.InitializeParticleFromNDC_FortWaterInteraction.AmmoType
      Float Selection Array=DataInterface() {propertyValues {FloatData [0.8,1,1.5,0.75,1]}}
     )
     B=$Output.InitializeParticleFromNDC_FortWaterInteraction.ImpactScaleMultiplier
    )) {isSetParameters=true}
   &InitializeParticle(
    Color=&MakeLinearColor(
     R:NiagaraFloat=&MakeFloatFromLinearColor(LinearColor=$Particles.HSVOffset, Channel:ENiagaraFloatFromLinearColorOptions=::NewEnumerator0)
     G:NiagaraFloat=&MakeFloatFromLinearColor(LinearColor=$Particles.HSVOffset, Channel:ENiagaraFloatFromLinearColorOptions=::NewEnumerator1)
     B:NiagaraFloat=&MakeFloatFromLinearColor(LinearColor=$Particles.HSVOffset, Channel:ENiagaraFloatFromLinearColorOptions=::NewEnumerator2)
     A:NiagaraFloat=1
    )
    Lifetime=&UniformRangedFloat(Minimum=0.2, Maximum=1)
    Mass Mode=::NewEnumerator1
    Position=$Output.InitializeParticleFromNDC_FortWaterInteraction.ImpactPoint
    Position Mode=::NewEnumerator1
    Sprite Size Mode=::NewEnumerator1
    Uniform Sprite Size Max=6
    UsePositionOffset=false
    Write Material Random=true
   )
   &AddVelocityInCone(
    Cone Angle=5
    Cone Axis=$Output.InitializeParticleFromNDC_FortWaterInteraction.ImpactNormal
    Module.Use Velocity Falloff On Cone Axis=false
    Velocity Strength=&Multiply_Float(
     A=&Multiply_Float(A=&UniformRangedFloat(Minimum=425, Maximum=450), B=$Particles.AmmoTypeScale)
     B=&FloatFromCurve(
      FloatCurve=DataInterface() {propertyValues {Curve {keys [{arriveTangent=0.6429, interpMode=RCIM_Cubic, leaveTangent=0.6429, tangentMode=RCTM_User, time=0, value=0.5}, <at 1 1 RCIM_Cubic>]}}}
      CurveIndex=&ReturnNormalizedExecIndex()
     )
    )
   )
  ]
  particleUpdateScript [
   &ParticleState()
   &GravityForce(Gravity=&Multiply_VectorByFloat(Vector=<0 0 -980>, Float=1))
   &SolveForcesAndVelocity(Clamp Velocity=false, Limit Acceleration=false)
   &ScaleSpriteSize(
    Uniform Curve Scale=&FloatFromCurve(
     FloatCurve=DataInterface() {propertyValues {Curve {keys [{arriveTangent=1.1621, interpMode=RCIM_Cubic, time=0, value=0.66}, <at 1 1 RCIM_Cubic>]}}}
     CurveIndex=&ReturnNormalizedExecIndex()
    )
    Uniform Curve Sprite Scale=DataInterface() {propertyValues {Curve {keys [
        {tangentMode=RCTM_User, time=0, value=1}
        {arriveTangent=-0.0405, interpMode=RCIM_Cubic, leaveTangent=-0.0405, tangentMode=RCTM_User, time=0.7009, value=1}
        {arriveTangent=-8.0089, interpMode=RCIM_Cubic, leaveTangent=-8.0089, tangentMode=RCTM_User, time=1, value=0}
       ]}}}
   )
   &DynamicMaterialParameters(
    Index 0 Param 2=&FloatFromCurve(FloatCurve=DataInterface() {propertyValues {Curve {keys [{interpMode=RCIM_Cubic, tangentMode=RCTM_User, time=0, value=1}, {interpMode=RCIM_Cubic, tangentMode=RCTM_User, time=1, value=0}]}}})
    Module.Index0Param0Write=true
    Module.Index0Param1Write=true
    Module.Index0Param2Write=true
    Module.Index0Param3Write=true
   )
   &Color(Color=&MakeLinearColorFromVectorAndFloat(
     "Vector (RGB)"=&MakeVectorFromLinearColorRGB(Color=$Particles.Color)
     "Float (Alpha)"=&FloatFromCurve(FloatCurve=DataInterface() {propertyValues {Curve {keys [<at 0.4 1.0113 RCIM_Cubic>,<at 1.0011 0.002>]}}})
    ))
  ]
  rendererClasses [&NiagaraSpriteRendererProperties]
  renderers [{rendererClass=&NiagaraSpriteRendererProperties}]
 }
 Drops_Tall_Splashy {
  emitterUpdateScript [
   &EmitterState()
   &SpawnBurst_Instantaneous(Spawn Count=&UniformRangedInt(Minimum=5, Maximum=7), Use Spawn Probability=false) {enabled=false}
   &Spawn_NDCFortImpacts(Max Spawn Count=20, Min Spawn Count=10, Reader:NiagaraDataInterfaceDataChannelRead=$System.NDCReader)
  ]
  particleSpawnScript [
   &InitializeParticleFromNDC_FortWaterInteraction()
   &HSV_Offset(Input Color:LinearColor=&MakeCustomLinearColorFromBool(Bool=$System.DoDebugColor, True Color=$System.DebugWaterColor, False Color=$Output.InitializeParticleFromNDC_FortWaterInteraction.WaterColor))
   /SetVariables(Particles.AmmoTypeScale:NiagaraFloat=&Multiply_Float(
     A=&SelectFloatFromArray(
      Direct Array Index=$Output.InitializeParticleFromNDC_FortWaterInteraction.AmmoType
      Float Selection Array=DataInterface() {propertyValues {FloatData [0.8,1,1.5,0.75,1]}}
     )
     B=$Output.InitializeParticleFromNDC_FortWaterInteraction.ImpactScaleMultiplier
    )) {isSetParameters=true, module=SetVariables_#4}
   /Niagara/Modules/Spawn/Initialization/InitializeParticle(
    Color=&MakeLinearColor(
     R:NiagaraFloat=&MakeFloatFromLinearColor(LinearColor=$Particles.HSVOffset, Channel:ENiagaraFloatFromLinearColorOptions=::NewEnumerator0)
     G:NiagaraFloat=&MakeFloatFromLinearColor(LinearColor=$Particles.HSVOffset, Channel:ENiagaraFloatFromLinearColorOptions=::NewEnumerator1)
     B:NiagaraFloat=&MakeFloatFromLinearColor(LinearColor=$Particles.HSVOffset, Channel:ENiagaraFloatFromLinearColorOptions=::NewEnumerator2)
     A:NiagaraFloat=1
    )
    Lifetime=&UniformRangedFloat(Minimum=0.06, Maximum=1)
    Module.Write Color=true
    Module.Write Lifetime=true
    Module.Write Mass=true
    Module.Write Material Random=true
    Module.Write Position=true
    Module.Write Scale=false
    Module.Write SpriteRotation=false
    Module.Write SpriteSize=true
    Position=$Output.InitializeParticleFromNDC_FortWaterInteraction.ImpactPoint
    Sprite Size=&Multiply_Vector2D_ByFloat(Float=2, Vector2D=&UniformRangedVector2D(Minimum=<3 4>, Maximum=<7 8>))
   ) {enabled=false}
   /Niagara/Modules/Spawn/Initialization/V2/InitializeParticle(
    Color=&MakeLinearColor(
     R:NiagaraFloat=&MakeFloatFromLinearColor(LinearColor=$Particles.HSVOffset, Channel:ENiagaraFloatFromLinearColorOptions=::NewEnumerator0)
     G:NiagaraFloat=&MakeFloatFromLinearColor(LinearColor=$Particles.HSVOffset, Channel:ENiagaraFloatFromLinearColorOptions=::NewEnumerator1)
     B:NiagaraFloat=&MakeFloatFromLinearColor(LinearColor=$Particles.HSVOffset, Channel:ENiagaraFloatFromLinearColorOptions=::NewEnumerator2)
     A:NiagaraFloat=1
    )
    Lifetime=&UniformRangedFloat(Minimum=0.15, Maximum=0.3)
    Mass Mode=::NewEnumerator1
    Position=$Output.InitializeParticleFromNDC_FortWaterInteraction.ImpactPoint
    Position Mode=::NewEnumerator1
    Sprite Size=&Multiply_Vector2D_ByFloat(Float=1, Vector2D=&UniformRangedVector2D(Minimum=<3 4>, Maximum=<7 8>))
    Sprite Size Mode=::NewEnumerator4
    UsePositionOffset=false
    Write Material Random=true
   ) {module=InitializeParticle001}
   &AddVelocityInCone(
    Cone Angle=5
    Cone Axis=$Output.InitializeParticleFromNDC_FortWaterInteraction.ImpactNormal
    Module.Use Velocity Falloff On Cone Axis=false
    Velocity Strength=&UniformRangedFloat(Minimum=200, Maximum=1500)
   )
   &FN_PerParticleLOD(Enable SpriteSize Distance Scaling=true, Sprite Size Scale Far=<1.5 1.5>)
  ]
  particleUpdateScript [
   &ParticleState()
   &GravityForce(Gravity=&Multiply_VectorByFloat(Vector=<0 0 -980>, Float=1))
   &SolveForcesAndVelocity(Clamp Velocity=false, Limit Acceleration=false)
   &ScaleSpriteSize(Scale Factor=&Vector2DFromFloat(Value=&FloatFromCurve(FloatCurve=DataInterface() {propertyValues {Curve {keys [<at 0.5 1.0113>,<at 1.0011 0.002>]}}})))
   /Niagara/Modules/Update/SubUV/SubUVAnimation(Random SubUV Lookup Index:NiagaraFloat=$Particles.Initial.MaterialRandom, SubUV Animation Mode=::NewEnumerator1) {enabled=false}
   &DynamicMaterialParameters(
    Index 0 Param 2=&FloatFromCurve(FloatCurve=DataInterface() {propertyValues {Curve {keys [{interpMode=RCIM_Cubic, tangentMode=RCTM_User, time=0, value=1}, {interpMode=RCIM_Cubic, tangentMode=RCTM_User, time=0.33, value=0}]}}})
    Index 0 Param 4=&FloatFromCurve(FloatCurve=DataInterface() {propertyValues {Curve {keys [{tangentMode=RCTM_User, time=0, value=0}, {tangentMode=RCTM_User, time=1, value=1}]}}})
    Module.Param0WriteEnabled=false
    Module.Param1WriteEnabled=true
    Module.Param2WriteEnabled=false
    Module.Param3WriteEnabled=false
   )
   &Color(Color=&MakeLinearColorFromVectorAndFloat(
     "Vector (RGB)"=&MakeVectorFromLinearColorRGB(Color=$Particles.Color)
     "Float (Alpha)"=&FloatFromCurve(FloatCurve=DataInterface() {propertyValues {Curve {keys [<at 0.4 1.0113 RCIM_Cubic>,<at 1.0011 0.002>]}}})
    ))
  ]
  rendererClasses [&NiagaraSpriteRendererProperties]
  renderers [{rendererClass=&NiagaraSpriteRendererProperties}]
 }
 Drops_Wide001 {
  bEnabled=false
  emitterUpdateScript [
   &EmitterState(
    Enable Distance Culling=true
    Max Distance Response=::KillAfterParticlesFinish
    Module.bUseMaxDistance=true
    Module.bUseMinDistance=false
    Scalability Mode=::NewEnumerator1
   )
   &Spawn_NDCFortImpacts(Max Spawn Count=13, Min Spawn Count=9, Reader:NiagaraDataInterfaceDataChannelRead=$System.NDCReader)
  ]
  particleSpawnScript [
   &InitializeParticleFromNDC_FortWaterInteraction()
   &HSV_Offset(Input Color:LinearColor=&MakeCustomLinearColorFromBool(Bool=$System.DoDebugColor, True Color=$System.DebugWaterColor, False Color=$Output.InitializeParticleFromNDC_FortWaterInteraction.WaterColor))
   /SetVariables(Particles.AmmoTypeScale:NiagaraFloat=&Multiply_Float(
     A=&SelectFloatFromArray(
      Direct Array Index=$Output.InitializeParticleFromNDC_FortWaterInteraction.AmmoType
      Float Selection Array=DataInterface() {propertyValues {FloatData [0.8,1,1.5,0.75,1]}}
     )
     B=$Output.InitializeParticleFromNDC_FortWaterInteraction.ImpactScaleMultiplier
    )) {isSetParameters=true}
   &InitializeParticle(
    Color=&MakeLinearColor(
     R:NiagaraFloat=&MakeFloatFromLinearColor(LinearColor=$Particles.HSVOffset, Channel:ENiagaraFloatFromLinearColorOptions=::NewEnumerator0)
     G:NiagaraFloat=&MakeFloatFromLinearColor(LinearColor=$Particles.HSVOffset, Channel:ENiagaraFloatFromLinearColorOptions=::NewEnumerator1)
     B:NiagaraFloat=&MakeFloatFromLinearColor(LinearColor=$Particles.HSVOffset, Channel:ENiagaraFloatFromLinearColorOptions=::NewEnumerator2)
     A:NiagaraFloat=1
    )
    Lifetime=&UniformRangedFloat(Minimum=0.2, Maximum=0.6)
    Position=$Output.InitializeParticleFromNDC_FortWaterInteraction.ImpactPoint
    Position Mode=::NewEnumerator1
    Sprite Size Mode=::NewEnumerator1
    Uniform Sprite Size Max=5
    Uniform Sprite Size Min=2
    UsePositionOffset=false
    Write Material Random=true
   )
   &AddVelocityInCone(
    Cone Angle=&RandomRangeFloat(Minimum=45, Maximum=70)
    Cone Axis=$Output.InitializeParticleFromNDC_FortWaterInteraction.ImpactNormal
    Module.Use Velocity Falloff On Cone Axis=false
    Velocity Strength=&Multiply_Float(A=&UniformRangedFloat(Minimum=250, Maximum=300), B=$Particles.AmmoTypeScale)
   )
  ]
  particleUpdateScript [
   &ParticleState()
   &AccelerationForce(Acceleration=<0 0 -980>)
   &SolveForcesAndVelocity(Clamp Velocity=false, Limit Acceleration=false)
   &ScaleSpriteSize(Uniform Curve Sprite Scale=DataInterface() {propertyValues {Curve {keys [
        {tangentMode=RCTM_User, time=0, value=1}
        {arriveTangent=-0.0405, interpMode=RCIM_Cubic, leaveTangent=-0.0405, tangentMode=RCTM_User, time=0.7009, value=1}
        {arriveTangent=-8.0089, interpMode=RCIM_Cubic, leaveTangent=-8.0089, tangentMode=RCTM_User, time=1, value=0}
       ]}}})
   &ScaleSpriteSizeBySpeed(Max Scale Factor=<0.5 4>, Velocity Threshold=500)
   &DynamicMaterialParameters(
    Index 0 Param 2=&FloatFromCurve(FloatCurve=DataInterface() {propertyValues {Curve {keys [{interpMode=RCIM_Cubic, tangentMode=RCTM_User, time=0, value=1}, {interpMode=RCIM_Cubic, tangentMode=RCTM_User, time=1, value=0}]}}})
    Module.Index0Param0Write=true
    Module.Index0Param1Write=true
    Module.Index0Param2Write=true
    Module.Index0Param3Write=true
   )
   &Color(Color=&MakeLinearColorFromVectorAndFloat(
     "Vector (RGB)"=&MakeVectorFromLinearColorRGB(Color=$Particles.Color)
     "Float (Alpha)"=&FloatFromCurve(FloatCurve=DataInterface() {propertyValues {Curve {keys [<at 0.4 1.0113 RCIM_Cubic>,<at 1.0011 0.002>]}}})
    ))
  ]
  rendererClasses [&NiagaraSpriteRendererProperties]
  renderers [{rendererClass=&NiagaraSpriteRendererProperties}]
 }
 Drops_Wide_Splashy {
  simTarget=GPUComputeSim
  emitterUpdateScript [
   &EmitterState()
   &Spawn_NDCFortImpacts(Max Spawn Count=60, Min Spawn Count=30, Reader:NiagaraDataInterfaceDataChannelRead=$System.NDCReader)
  ]
  particleSpawnScript [
   &InitializeParticleFromNDC_FortWaterInteraction()
   &HSV_Offset(Input Color:LinearColor=&MakeCustomLinearColorFromBool(Bool=$System.DoDebugColor, True Color=$System.DebugWaterColor, False Color=$Output.InitializeParticleFromNDC_FortWaterInteraction.WaterColor))
   /SetVariables(Particles.AmmoTypeScale:NiagaraFloat=&Multiply_Float(
     A=&SelectFloatFromArray(
      Direct Array Index=$Output.InitializeParticleFromNDC_FortWaterInteraction.AmmoType
      Float Selection Array=DataInterface() {propertyValues {FloatData [0.8,1,1.5,0.75,1]}}
     )
     B=$Output.InitializeParticleFromNDC_FortWaterInteraction.ImpactScaleMultiplier
    )) {isSetParameters=true, module=SetVariables_#4}
   /Niagara/Modules/Spawn/Initialization/InitializeParticle(
    Color=&MakeLinearColor(
     R:NiagaraFloat=&MakeFloatFromLinearColor(LinearColor=$Particles.HSVOffset, Channel:ENiagaraFloatFromLinearColorOptions=::NewEnumerator0)
     G:NiagaraFloat=&MakeFloatFromLinearColor(LinearColor=$Particles.HSVOffset, Channel:ENiagaraFloatFromLinearColorOptions=::NewEnumerator1)
     B:NiagaraFloat=&MakeFloatFromLinearColor(LinearColor=$Particles.HSVOffset, Channel:ENiagaraFloatFromLinearColorOptions=::NewEnumerator2)
     A:NiagaraFloat=1
    )
    Lifetime=&UniformRangedFloat(Minimum=0.06, Maximum=1)
    Module.Write Color=true
    Module.Write Lifetime=true
    Module.Write Mass=true
    Module.Write Material Random=true
    Module.Write Position=true
    Module.Write Scale=false
    Module.Write SpriteRotation=false
    Module.Write SpriteSize=true
    Position=$Output.InitializeParticleFromNDC_FortWaterInteraction.ImpactPoint
    Sprite Size=&Multiply_Vector2D_ByFloat(Float=2, Vector2D=&UniformRangedVector2D(Minimum=<3 4>, Maximum=<7 8>))
   ) {enabled=false}
   /Niagara/Modules/Spawn/Initialization/V2/InitializeParticle(
    Color=&MakeLinearColor(
     R:NiagaraFloat=&MakeFloatFromLinearColor(LinearColor=$Particles.HSVOffset, Channel:ENiagaraFloatFromLinearColorOptions=::NewEnumerator0)
     G:NiagaraFloat=&MakeFloatFromLinearColor(LinearColor=$Particles.HSVOffset, Channel:ENiagaraFloatFromLinearColorOptions=::NewEnumerator1)
     B:NiagaraFloat=&MakeFloatFromLinearColor(LinearColor=$Particles.HSVOffset, Channel:ENiagaraFloatFromLinearColorOptions=::NewEnumerator2)
     A:NiagaraFloat=1
    )
    Lifetime=&UniformRangedFloat(Minimum=0.3, Maximum=1)
    Mass Mode=::NewEnumerator1
    Position=$Output.InitializeParticleFromNDC_FortWaterInteraction.ImpactPoint
    Position Mode=::NewEnumerator1
    Sprite Size=&Multiply_Vector2D_ByFloat(Float=1, Vector2D=&UniformRangedVector2D(Minimum=<3 4>, Maximum=<7 8>))
    Sprite Size Mode=::NewEnumerator4
    UsePositionOffset=false
    Write Material Random=true
   ) {module=InitializeParticle001}
   &AddVelocityInCone(
    Cone Angle=84.9615
    Cone Axis=$Output.InitializeParticleFromNDC_FortWaterInteraction.ImpactNormal
    Module.Use Velocity Falloff On Cone Axis=false
    Velocity Strength=&UniformRangedFloat(Minimum=200, Maximum=588.3452)
   )
   &FN_PerParticleLOD(Enable SpriteSize Distance Scaling=true, Sprite Size Scale Far=<1.5 1.5>)
  ]
  particleUpdateScript [
   &ParticleState()
   &GravityForce(Gravity=&Multiply_VectorByFloat(Vector=<0 0 -980>, Float=1))
   &SolveForcesAndVelocity(Clamp Velocity=false, Limit Acceleration=false)
   &ScaleSpriteSize(Scale Factor=&Vector2DFromFloat(Value=&OneMinusFloat(Float=$Particles.NormalizedAge)))
   &ScaleSpriteSizeBySpeed(Max Scale Factor=<1 2>, Velocity Threshold=500)
   &SubUVAnimation(Random SubUV Lookup Index:NiagaraFloat=$Particles.Initial.MaterialRandom, SubUV Animation Mode=::NewEnumerator1) {enabled=false}
   &DynamicMaterialParameters(
    Index 0 Param 2=&FloatFromCurve(FloatCurve=DataInterface() {propertyValues {Curve {keys [{interpMode=RCIM_Cubic, tangentMode=RCTM_User, time=0, value=1}, {interpMode=RCIM_Cubic, tangentMode=RCTM_User, time=0.33, value=0}]}}})
    Index 0 Param 4=&FloatFromCurve(FloatCurve=DataInterface() {propertyValues {Curve {keys [{tangentMode=RCTM_User, time=0, value=0}, {tangentMode=RCTM_User, time=1, value=1}]}}})
    Module.Param0WriteEnabled=false
    Module.Param1WriteEnabled=true
    Module.Param2WriteEnabled=false
    Module.Param3WriteEnabled=false
   ) {enabled=false}
   &Color(Color=&MakeLinearColorFromVectorAndFloat(
     "Vector (RGB)"=&MakeVectorFromLinearColorRGB(Color=$Particles.Color)
     "Float (Alpha)"=&FloatFromCurve(FloatCurve=DataInterface() {propertyValues {Curve {keys [<at 0.4 1.0113 RCIM_Cubic>,<at 1.0011 0.002>]}}})
    ))
  ]
  rendererClasses [&NiagaraSpriteRendererProperties]
  renderers [{rendererClass=&NiagaraSpriteRendererProperties}]
 }
 FN_Weapons_Debris_Sprite_SubUV_BETA {
  emitterUpdateScript [
   &EmitterState()
   &Spawn_NDCFortImpacts(Max Spawn Count=1, Min Spawn Count=1, Reader:NiagaraDataInterfaceDataChannelRead=$System.NDCReader)
  ]
  particleSpawnScript [
   &InitializeParticleFromNDC_FortWaterInteraction()
   &HSV_Offset(Input Color:LinearColor=&MakeCustomLinearColorFromBool(Bool=$System.DoDebugColor, True Color=$System.DebugWaterColor, False Color=$Output.InitializeParticleFromNDC_FortWaterInteraction.WaterColor))
   &InitializeParticle(
    Color=&MakeLinearColor(
     R:NiagaraFloat=&MakeFloatFromLinearColor(LinearColor=$Particles.HSVOffset, Channel:ENiagaraFloatFromLinearColorOptions=::NewEnumerator0)
     G:NiagaraFloat=&MakeFloatFromLinearColor(LinearColor=$Particles.HSVOffset, Channel:ENiagaraFloatFromLinearColorOptions=::NewEnumerator1)
     B:NiagaraFloat=&MakeFloatFromLinearColor(LinearColor=$Particles.HSVOffset, Channel:ENiagaraFloatFromLinearColorOptions=::NewEnumerator2)
     A:NiagaraFloat=1
    )
    Lifetime=0.3
    Mass Mode=::NewEnumerator1
    Position=&Add_Vector(
     A=Linked() {linkedVariable {name=Output.InitializeParticleFromNDC_FortWaterInteraction.ImpactPoint, type=NiagaraPosition}}
     B=&Multiply_VectorByFloat(
      Vector=&NormalizeVector(
       Vector To Normalize=$Output.InitializeParticleFromNDC_FortWaterInteraction.ImpactNormal
       Safe Normalize=true
       Safe Fallback Threshold=0
       Safe Fallback Vector=<0 0 0.001>
      )
      Float=2
     )
    )
    Position Mode=::NewEnumerator1
    Sprite Size=&Multiply_Vector2D_ByFloat(Float=1, Vector2D=&RandomRangeVector2D(Minimum=<6 10>, Maximum=<12 16>))
    Sprite Size Mode=::NewEnumerator4
    UsePositionOffset=false
    Write Material Random=true
   ) {module=InitializeParticle001}
   &AddVelocityInCone(
    Cone Angle=0
    Cone Axis=$Output.InitializeParticleFromNDC_FortWaterInteraction.ImpactNormal
    Module.Use Velocity Falloff On Cone Axis=false
    Velocity Strength=&UniformRangedFloat(Minimum=100, Maximum=300)
   )
   &FN_PerParticleLOD(Enable SpriteSize Distance Scaling=true, Sprite Size Scale Far=<1.5 1.5>)
   /NiagaraFN/Sprite/FN_RandomizeSpriteUV()
   /Niagara/Modules/Update/SubUV/V2/SubUVAnimation(
    Random Start Frame=true
    "Sprite Renderer "=DataInterface() {propertyValues {SpriteRenderer=&self:FN_Weapons_Debris_Sprite_SubUV_BETA_0.NiagaraSpriteRendererProperties_0}}
    UseEndFrame=false
    UseStartFrame=false
   ) {enabled=false, module=SubUVAnimation001}
   /SetVariables(Particles.Lifetime=&MaskFloat(
     Mask Source=1
     Mask Parameter=$Output.InitializeParticleFromNDC_FortWaterInteraction.AmmoType
     Mask=3
     Masked Float=&Multiply_Float(A=$Particles.Lifetime, B=&RandomRangeFloat(Minimum=0.8, Maximum=1.1))
     Passthrough Float=$Particles.Lifetime
    )) {isSetParameters=true}
  ]
  particleUpdateScript [
   &ParticleState()
   &GravityForce(Gravity=&Multiply_VectorByFloat(Vector=<0 0 -800>, Float=1))
   &Drag(Drag=0.1, Use Linear Drag=true, Use Rotational Drag=false) {enabled=false}
   &SolveForcesAndVelocity(Clamp Velocity=false, Limit Acceleration=false)
   &ScaleSpriteSize(Scale Factor=&Vector2DFromFloat(Value=&FloatFromCurve(FloatCurve=DataInterface() {propertyValues {Curve {keys [<at -0.0046 1.0113>,<at 0.77 7.78>,<at 1.0011 0.002>]}}})))
   &DynamicMaterialParameters(
    Index 0 Param 2=&FloatFromCurve(FloatCurve=DataInterface() {propertyValues {Curve {keys [{interpMode=RCIM_Cubic, tangentMode=RCTM_User, time=0, value=1}, {interpMode=RCIM_Cubic, tangentMode=RCTM_User, time=1, value=0}]}}})
    Index 0 Param 4=&FloatFromCurve(FloatCurve=DataInterface() {propertyValues {Curve {keys [{tangentMode=RCTM_User, time=0, value=0}, {tangentMode=RCTM_User, time=1, value=1}]}}})
    Module.Param0WriteEnabled=false
    Module.Param1WriteEnabled=true
    Module.Param2WriteEnabled=false
    Module.Param3WriteEnabled=true
   ) {enabled=false}
   &DynamicMaterialParameters(
    Index 0 Param 2=&FloatFromCurve(FloatCurve=DataInterface() {propertyValues {Curve {keys [
         {arriveTangent=46.3377, interpMode=RCIM_Cubic, leaveTangent=46.3379, tangentMode=RCTM_User, time=-0.0046, value=0}
         {arriveTangent=0.0019, interpMode=RCIM_Cubic, leaveTangent=0.0019, time=0.05, value=0.9763}
         {arriveTangent=-2.5858, interpMode=RCIM_Cubic, leaveTangent=-2.5858, tangentMode=RCTM_User, time=1.0011, value=0.002}
        ]}}})
    Module.Param0WriteEnabled=false
    Module.Param1WriteEnabled=true
    Module.Param2WriteEnabled=false
    Module.Param3WriteEnabled=false
   ) {module=DynamicMaterialParameters001}
   /Niagara/Modules/Update/SubUV/SubUVAnimation(Random SubUV Lookup Index:NiagaraFloat=$Particles.Initial.MaterialRandom, SubUV Animation Mode=::NewEnumerator1) {enabled=false}
   &Color(Color=&MakeLinearColorFromVectorAndFloat(
     "Vector (RGB)"=&MakeVectorFromLinearColorRGB(Color=$Particles.Color)
     "Float (Alpha)"=&FloatFromCurve(
      FloatCurve=DataInterface() {propertyValues {Curve {keys [{arriveTangent=0.045, interpMode=RCIM_Cubic, leaveTangent=0.045, tangentMode=RCTM_User, time=0.4, value=1}, <at 1.0011 0.002>]}}}
      Scale Curve=0.01
     )
    ))
  ]
  rendererClasses [&NiagaraSpriteRendererProperties]
  renderers [{rendererClass=&NiagaraSpriteRendererProperties}]
 }
 Impact_Surface_Foam {
  bEnabled=false
  emitterUpdateScript [
   &EmitterState()
   &Spawn_NDCFortImpacts(Max Spawn Count=1, Min Spawn Count=1, Reader:NiagaraDataInterfaceDataChannelRead=$System.NDCReader)
   /Game/Effects/Niagara/DataChannels/Impacts/SimpleFilter_NDCFortImpacts(Exclude=false, Max Spawn Count=1) {enabled=false}
   /Game/Effects/Niagara/DataChannels/Impacts/LODDistanceFilter_NDCFortImpacts(LODDistance=6000) {enabled=false}
  ]
  particleSpawnScript [
   &InitializeParticleFromNDC_FortWaterInteraction(FallbackScaleMultiplier=50)
   &HSV_Offset(Input Color:LinearColor=&MakeCustomLinearColorFromBool(Bool=$System.DoDebugColor, True Color=$System.DebugWaterColor, False Color=$Output.InitializeParticleFromNDC_FortWaterInteraction.WaterColor))
   /SetVariables(Particles.AmmoTypeScale:NiagaraFloat=&Multiply_Float(
     A=&SelectFloatFromArray(
      Direct Array Index=$Output.InitializeParticleFromNDC_FortWaterInteraction.AmmoType
      Float Selection Array=DataInterface() {propertyValues {FloatData [0.6,0.8,1.2,0.4,1]}}
     )
     B=$Output.InitializeParticleFromNDC_FortWaterInteraction.ImpactScaleMultiplier
    )) {isSetParameters=true, module=SetVariables_#5}
   &InitializeParticle(
    Color=&MakeLinearColor(
     R:NiagaraFloat=&MakeFloatFromLinearColor(LinearColor=$Particles.HSVOffset, Channel:ENiagaraFloatFromLinearColorOptions=::NewEnumerator0)
     G:NiagaraFloat=&MakeFloatFromLinearColor(LinearColor=$Particles.HSVOffset, Channel:ENiagaraFloatFromLinearColorOptions=::NewEnumerator1)
     B:NiagaraFloat=&MakeFloatFromLinearColor(LinearColor=$Particles.HSVOffset, Channel:ENiagaraFloatFromLinearColorOptions=::NewEnumerator2)
     A:NiagaraFloat=1
    )
    Lifetime Max=1.6
    Lifetime Min=1.5
    Lifetime Mode=::NewEnumerator1
    Mass=&UniformRangedFloat(Minimum=0.6, Maximum=1.2)
    Mass Mode=::NewEnumerator1
    Mesh Scale Mode=::NewEnumerator3
    Mesh Uniform Scale=&Multiply_Float(A=200, B=$Particles.AmmoTypeScale)
    Position=$Output.InitializeParticleFromNDC_FortWaterInteraction.ImpactPoint
    Position Mode=::NewEnumerator1
    Position Offset=&Add_Vector(A=$Output.InitializeParticleFromNDC_FortWaterInteraction.WaveOffset, B=<0 0 2>)
    Position Offset Coordinate Space=::World
    UsePositionOffset=true
    Write Material Random=true
   )
   &InitialMeshOrientation(
    Mesh Orientation Mode=::NewEnumerator2
    Orientation Vector=<0 0 0>
    Rotation=&UniformRangedVector(Minimum=<0 0 0>, Maximum=<0 0 1>)
    Use Rotation Vector=false
   ) {module=InitialMeshOrientation001}
   &DynamicMaterialParameters(
    Index 3 Param 1=&MakeFloatFromInt(Int=$Output.InitializeParticleFromNDC_FortWaterInteraction.WaterBodyIndex)
    Index 3 Param 2=$Output.InitializeParticleFromNDC_FortWaterInteraction.WaterDepth
    Index 3 Param 3=$Output.InitializeParticleFromNDC_FortWaterInteraction.WaveAttenuationFactor
    Module.Index3Param0Write=true
    Module.Index3Param1Write=true
    Module.Index3Param2Write=true
    Module.Index3Param3Write=true
    Write Parameter Index 0=false
    Write Parameter Index 3=true
   )
  ]
  particleUpdateScript [
   &ParticleState()
   &DynamicMaterialParameters(
    Index 0 Param 1=&RandomRangeFloat(Minimum=-1, Maximum=1)
    Index 0 Param 2=&FloatFromCurve(FloatCurve=DataInterface() {propertyValues {Curve {keys [<at 0 0.33>,<at 1 1.25>]}}})
    Index 0 Param 3=&FloatFromCurve(
     FloatCurve=DataInterface() {propertyValues {Curve {keys [{tangentMode=RCTM_User, time=0, value=1}, {tangentMode=RCTM_User, time=1, value=0}]}}}
     Scale Curve=1.2
    )
    Index 1 Param 1=&MakeFloatFromVector(Vector=Linked() {linkedVariable {name=Particles.Initial.Position, type=NiagaraPosition}}, Channel=::NewEnumerator0)
    Index 1 Param 2=&MakeFloatFromVector(Vector=Linked() {linkedVariable {name=Particles.Initial.Position, type=NiagaraPosition}}, Channel=::NewEnumerator1)
    Index 1 Param 3=&FloatFromCurve(
     FloatCurve=DataInterface() {propertyValues {Curve {keys [<at 0 0.8>, {arriveTangent=0.579, leaveTangent=0.579, time=1, value=1}]}}}
     Scale Curve=1600
    )
    Index 2 Param 1=&FloatFromCurve(FloatCurve=DataInterface() {propertyValues {Curve {keys [{interpMode=RCIM_Cubic, tangentMode=RCTM_User, time=0, value=0.4}, {interpMode=RCIM_Cubic, tangentMode=RCTM_User, time=1, value=1}]}}})
    Index 2 Param 2=&FloatFromCurve(FloatCurve=DataInterface() {propertyValues {Curve {keys [<at 0 2>,<at 1 0>]}}})
    Index 2 Param 3=1
    Index 2 Param 4=&FloatFromCurve(FloatCurve=DataInterface() {propertyValues {Curve {keys [
         {arriveTangent=-1.9991, interpMode=RCIM_Cubic, time=0, value=1}
         {arriveTangent=-2.7143, interpMode=RCIM_Cubic, leaveTangent=-2.7143, tangentMode=RCTM_User, time=0.2095, value=0.6901}
         <at 1 0 RCIM_Cubic>
        ]}}})
    Module.Index0Param0Write=true
    Module.Index0Param1Write=true
    Module.Index0Param2Write=true
    Module.Index0Param3Write=true
    Module.Index1Param0Write=true
    Module.Index1Param1Write=true
    Module.Index1Param2Write=true
    Module.Index1Param3Write=true
    Module.Index2Param0Write=true
    Module.Index2Param1Write=true
    Module.Index2Param2Write=true
    Module.Index2Param3Write=true
    Write Parameter Index 1=true
    Write Parameter Index 2=true
   ) {module=DynamicMaterialParameters001}
   &ScaleMeshSize(Scale Factor=&VectorFromFloat(Value=&FloatFromCurve(FloatCurve=DataInterface() {propertyValues {Curve {keys [{arriveTangent=0.9849, interpMode=RCIM_Cubic, leaveTangent=0.9849, tangentMode=RCTM_User, time=0, value=0.33}, <at 1 1 RCIM_Cubic>]}}}))) {module=MeshSizeScale}
   &Color(Color=&MakeLinearColorFromVectorAndFloat(
     "Vector (RGB)"=&MakeVectorFromLinearColorRGB(Color=$Particles.Color)
     "Float (Alpha)"=&FloatFromCurve(FloatCurve=DataInterface() {propertyValues {Curve {keys [<at 0.4 1.0113 RCIM_Cubic>,<at 1.0011 0.002>]}}})
    ))
   &ConstrainPositionToPlane(
    Plane Normal:Vector3f=$Output.InitializeParticleFromNDC_FortWaterInteraction.WaterPlaneNormal
    Plane Position:NiagaraPosition=$Output.InitializeParticleFromNDC_FortWaterInteraction.WaterPlaneLocation
   )
  ]
  rendererClasses [&NiagaraMeshRendererProperties]
  renderers [{rendererClass=&NiagaraMeshRendererProperties, rendererIndex=0}]
 }
 Impact_Surface_Foam001 {
  emitterUpdateScript [
   &EmitterState()
   &Spawn_NDCFortImpacts(Max Spawn Count=1, Min Spawn Count=1, Reader:NiagaraDataInterfaceDataChannelRead=$System.NDCReader)
   &SimpleFilter_NDCFortImpacts(Exclude=false, Max Spawn Count=1) {enabled=false}
   &LODDistanceFilter_NDCFortImpacts(LODDistance=6000) {enabled=false}
  ]
  particleSpawnScript [
   &InitializeParticleFromNDC_FortWaterInteraction(FallbackScaleMultiplier=50)
   &HSV_Offset(Input Color:LinearColor=&MakeCustomLinearColorFromBool(Bool=$System.DoDebugColor, True Color=$System.DebugWaterColor, False Color=$Output.InitializeParticleFromNDC_FortWaterInteraction.WaterColor))
   /SetVariables(Particles.AmmoTypeScale:NiagaraFloat=&Multiply_Float(
     A=&SelectFloatFromArray(
      Direct Array Index=$Output.InitializeParticleFromNDC_FortWaterInteraction.AmmoType
      Float Selection Array=DataInterface() {propertyValues {FloatData [0.6,0.8,1.2,1,1]}}
     )
     B=$Output.InitializeParticleFromNDC_FortWaterInteraction.ImpactScaleMultiplier
    )) {isSetParameters=true, module=SetVariables_#5}
   /SetVariables(Particles.AmmoTypeScale:NiagaraFloat=&Multiply_Float(
     A=$Particles.AmmoTypeScale
     B=&MaskFloat(
      Mask Source=1
      Mask Parameter=$Output.InitializeParticleFromNDC_FortWaterInteraction.AmmoType
      Mask=3
      Masked Float=&RandomRangeFloat(Minimum=0.3, Maximum=1.2)
      Passthrough Float=1
     )
    )) {isSetParameters=true}
   &InitializeParticle(
    Color=&MakeLinearColor(
     R:NiagaraFloat=&MakeFloatFromLinearColor(LinearColor=$Particles.HSVOffset, Channel:ENiagaraFloatFromLinearColorOptions=::NewEnumerator0)
     G:NiagaraFloat=&MakeFloatFromLinearColor(LinearColor=$Particles.HSVOffset, Channel:ENiagaraFloatFromLinearColorOptions=::NewEnumerator1)
     B:NiagaraFloat=&MakeFloatFromLinearColor(LinearColor=$Particles.HSVOffset, Channel:ENiagaraFloatFromLinearColorOptions=::NewEnumerator2)
     A:NiagaraFloat=1
    )
    Lifetime Max=1.6
    Lifetime Min=1.5
    Lifetime Mode=::NewEnumerator1
    Mass=&UniformRangedFloat(Minimum=0.6, Maximum=1.2)
    Mass Mode=::NewEnumerator1
    Mesh Scale Mode=::NewEnumerator3
    Mesh Uniform Scale=&Multiply_Float(A=350, B=$Particles.AmmoTypeScale)
    Position=$Output.InitializeParticleFromNDC_FortWaterInteraction.ImpactPoint
    Position Mode=::NewEnumerator1
    Position Offset=&Add_Vector(A=$Output.InitializeParticleFromNDC_FortWaterInteraction.WaveOffset, B=<0 0 2>)
    Position Offset Coordinate Space=::World
    UsePositionOffset=true
    Write Material Random=true
   )
   &InitialMeshOrientation(
    Mesh Orientation Mode=::NewEnumerator2
    Orientation Vector=<0 0 0>
    Rotation=&UniformRangedVector(Minimum=<0 0 0>, Maximum=<0 0 1>)
    Use Rotation Vector=false
   ) {module=InitialMeshOrientation001}
   &DynamicMaterialParameters(
    Index 3 Param 1=&MakeFloatFromInt(Int=$Output.InitializeParticleFromNDC_FortWaterInteraction.WaterBodyIndex)
    Index 3 Param 2=$Output.InitializeParticleFromNDC_FortWaterInteraction.WaterDepth
    Index 3 Param 3=$Output.InitializeParticleFromNDC_FortWaterInteraction.WaveAttenuationFactor
    Module.Index3Param0Write=true
    Module.Index3Param1Write=true
    Module.Index3Param2Write=true
    Module.Index3Param3Write=true
    Write Parameter Index 0=false
    Write Parameter Index 3=true
   )
  ]
  particleUpdateScript [
   &ParticleState()
   &DynamicMaterialParameters(
    Index 0 Param 1=&RandomRangeFloat(Minimum=-1, Maximum=1)
    Index 0 Param 2=&FloatFromCurve(FloatCurve=DataInterface() {propertyValues {Curve {keys [<at 0 0.33>,<at 1 1.25>]}}})
    Index 0 Param 3=&FloatFromCurve(
     FloatCurve=DataInterface() {propertyValues {Curve {keys [{tangentMode=RCTM_User, time=0, value=1}, {tangentMode=RCTM_User, time=1, value=0}]}}}
     Scale Curve=1.2
    )
    Index 1 Param 1=&MakeFloatFromVector(Vector=Linked() {linkedVariable {name=Particles.Initial.Position, type=NiagaraPosition}}, Channel=::NewEnumerator0)
    Index 1 Param 2=&MakeFloatFromVector(Vector=Linked() {linkedVariable {name=Particles.Initial.Position, type=NiagaraPosition}}, Channel=::NewEnumerator1)
    Index 1 Param 3=&FloatFromCurve(
     FloatCurve=DataInterface() {propertyValues {Curve {keys [<at 0 0.8>, {arriveTangent=0.579, leaveTangent=0.579, time=1, value=1}]}}}
     Scale Curve=1600
    )
    Index 2 Param 1=&FloatFromCurve(FloatCurve=DataInterface() {propertyValues {Curve {keys [{interpMode=RCIM_Cubic, tangentMode=RCTM_User, time=0, value=0.4}, {interpMode=RCIM_Cubic, tangentMode=RCTM_User, time=1, value=1}]}}})
    Index 2 Param 2=&FloatFromCurve(FloatCurve=DataInterface() {propertyValues {Curve {keys [<at 0 2>,<at 1 0>]}}})
    Index 2 Param 3=1
    Index 2 Param 4=&FloatFromCurve(FloatCurve=DataInterface() {propertyValues {Curve {keys [
         {arriveTangent=-1.9991, interpMode=RCIM_Cubic, time=0, value=1}
         {arriveTangent=-2.7143, interpMode=RCIM_Cubic, leaveTangent=-2.7143, tangentMode=RCTM_User, time=0.2095, value=0.6901}
         <at 1 0 RCIM_Cubic>
        ]}}})
    Module.Index0Param0Write=true
    Module.Index0Param1Write=true
    Module.Index0Param2Write=true
    Module.Index0Param3Write=true
    Module.Index1Param0Write=true
    Module.Index1Param1Write=true
    Module.Index1Param2Write=true
    Module.Index1Param3Write=true
    Module.Index2Param0Write=true
    Module.Index2Param1Write=true
    Module.Index2Param2Write=true
    Module.Index2Param3Write=true
    Write Parameter Index 1=true
    Write Parameter Index 2=true
   ) {module=DynamicMaterialParameters001}
   &ScaleMeshSize(Scale Factor=&VectorFromFloat(Value=&FloatFromCurve(FloatCurve=DataInterface() {propertyValues {Curve {keys [{arriveTangent=0.9849, interpMode=RCIM_Cubic, leaveTangent=0.9849, tangentMode=RCTM_User, time=0, value=0.33}, <at 1 1 RCIM_Cubic>]}}}))) {module=MeshSizeScale}
   &Color(Color=&MakeLinearColorFromVectorAndFloat(
     "Vector (RGB)"=&MakeVectorFromLinearColorRGB(Color=$Particles.Color)
     "Float (Alpha)"=&FloatFromCurve(FloatCurve=DataInterface() {propertyValues {Curve {keys [<at 0.4 1.0113 RCIM_Cubic>,<at 1.0011 0.002>]}}})
    ))
   &ConstrainPositionToPlane(
    Plane Normal:Vector3f=$Output.InitializeParticleFromNDC_FortWaterInteraction.WaterPlaneNormal
    Plane Position:NiagaraPosition=$Output.InitializeParticleFromNDC_FortWaterInteraction.WaterPlaneLocation
   )
  ]
  rendererClasses [&NiagaraMeshRendererProperties]
  renderers [{rendererClass=&NiagaraMeshRendererProperties, rendererIndex=0}]
 }
 Mist {
  bEnabled=false
  emitterUpdateScript [
   &EmitterState()
   &Spawn_NDCFortImpacts(Max Spawn Count=5, Min Spawn Count=4, Reader:NiagaraDataInterfaceDataChannelRead=$System.NDCReader)
  ]
  particleSpawnScript [
   &InitializeParticleFromNDC_FortWaterInteraction()
   &HSV_Offset(Input Color:LinearColor=$Output.InitializeParticleFromNDC_FortWaterInteraction.WaterColor)
   /SetVariables(Particles.AmmoTypeScale:NiagaraFloat=&Multiply_Float(
     A=&SelectFloatFromArray(
      Direct Array Index=$Output.InitializeParticleFromNDC_FortWaterInteraction.AmmoType
      Float Selection Array=DataInterface() {propertyValues {FloatData [0.8,1,1.5,0.75,1]}}
     )
     B=$Output.InitializeParticleFromNDC_FortWaterInteraction.ImpactScaleMultiplier
    )) {isSetParameters=true}
   /SetVariables(Particles.LifetimeMax=0.4, Particles.LifetimeMin=0.3) {isSetParameters=true}
   /SetVariables(Particles.SizeMax=60, Particles.SizeMin=80) {isSetParameters=true}
   /SetVariables(Particles.SizeCurve:NiagaraFloat=&Multiply_Float(
     A=&Lerp_Float(
      A=$Particles.SizeMin
      B=$Particles.SizeMax
      Alpha=&FloatFromCurve(
       FloatCurve=DataInterface() {propertyValues {Curve {keys [
           {interpMode=RCIM_Cubic, tangentMode=RCTM_Break, time=0, value=0}
           {arriveTangent=2.5778, interpMode=RCIM_Cubic, tangentMode=RCTM_Break, time=1, value=1}
          ]}}}
       CurveIndex=&ReturnNormalizedExecIndex()
      )
     )
     B=1
    )) {isSetParameters=true}
   &InitializeParticle(
    Color=<rgba 0.8963 0.9647 1 1>
    Lifetime=&Multiply_Float(
     A=&Lerp_Float(A=$Particles.LifetimeMin, B=$Particles.LifetimeMax, Alpha=&ReturnNormalizedExecIndex())
     B=$Particles.AmmoTypeScale
    )
    Position=$Output.InitializeParticleFromNDC_FortWaterInteraction.ImpactPoint
    Position Mode=::NewEnumerator1
    Sprite Rotation Mode=::NewEnumerator2
    Sprite Size Mode=::NewEnumerator3
    Uniform Sprite Size=&Multiply_Float(A=$Particles.SizeCurve, B=$Particles.AmmoTypeScale)
    UsePositionOffset=false
    Write Material Random=true
   )
   /SetVariables(Particles.InitSpriteRotRate=&UniformRangedFloat(Minimum=-75, Maximum=75)) {isSetParameters=true}
   &AddVelocityInCone(
    Cone Angle=10
    Cone Axis=$Output.InitializeParticleFromNDC_FortWaterInteraction.ImpactNormal
    Cone Axis Coordinate Space=::Simulation
    Module.Use Velocity Falloff On Cone Axis=false
    Velocity Strength=&Multiply_Float(
     A=&Lerp_Float(
      A=100
      B=450
      Alpha=&NormalizeFloat(Value=$Particles.SizeCurve, Min=$Particles.SizeMin, Max=$Particles.SizeMax)
     )
     B=$Particles.AmmoTypeScale
    )
   )
  ]
  particleUpdateScript [
   &ParticleState()
   &Drag(Drag=3, Use Linear Drag=true, Use Rotational Drag=false)
   &GravityForce()
   &SpriteRotationRate(Rotation Rate=&Multiply_Float(
     A=$Particles.InitSpriteRotRate
     B=&FloatFromCurve(FloatCurve=DataInterface() {propertyValues {Curve {keys [{arriveTangent=-2.0511, interpMode=RCIM_Cubic, leaveTangent=-2.0511, tangentMode=RCTM_User, time=0, value=1}, <at 1 0.1 RCIM_Cubic>]}}})
    ))
   &SolveForcesAndVelocity(Clamp Velocity=false, Limit Acceleration=false)
   &ScaleSpriteSize(Uniform Curve Sprite Scale=DataInterface() {propertyValues {Curve {keys [
        {arriveTangent=2.9194, interpMode=RCIM_Cubic, leaveTangent=2.9194, tangentMode=RCTM_User, time=0, value=0.25}
        {arriveTangent=1.1179, interpMode=RCIM_Cubic, leaveTangent=1.1179, tangentMode=RCTM_User, time=0.2, value=0.65}
        {arriveTangent=0.0224, interpMode=RCIM_Cubic, leaveTangent=0.0224, tangentMode=RCTM_User, time=1, value=1}
       ]}}})
   &ScaleColor(
    Scale Alpha=&FloatFromCurve(FloatCurve=DataInterface() {propertyValues {Curve {keys [
         {interpMode=RCIM_Cubic, leaveTangent=17.2181, tangentMode=RCTM_Break, time=0, value=0}
         {interpMode=RCIM_Cubic, tangentMode=RCTM_Break, time=0.1, value=1}
         {arriveTangent=-1.1111, interpMode=RCIM_Cubic, leaveTangent=-1.1111, time=0.5, value=0.2}
         <at 1 0 RCIM_Cubic>
        ]}}})
    ScaleA=true
    ScaleRGB=true
   )
   &CameraOffset(Camera Offset Amount=&Multiply_Float(A=&MakeFloatFromVector(Vector=$Particles.Scale, Channel=::NewEnumerator0), B=1))
   &DynamicMaterialParameters(
    Index 0 Param 1=&Add_Float(
     A=&RandomRangeFloat(Minimum=0, Maximum=1)
     B=&FloatFromCurve(
      FloatCurve=DataInterface() {propertyValues {Curve {keys [
          {arriveTangent=0.7934, interpMode=RCIM_Cubic, leaveTangent=0.7934, tangentMode=RCTM_User, time=0, value=1}
          {interpMode=RCIM_Cubic, leaveTangent=0.2021, time=1, value=1.5}
         ]}}}
      CurveIndex=&Add_Float(A=&Multiply_Float(A=$Particles.Initial.MaterialRandom, B=0.5), B=$Particles.NormalizedAge)
     )
    )
    Index 0 Param 4=&Lerp_Float(
     A=1
     B=0.75
     Alpha=&FloatFromCurve(FloatCurve=DataInterface() {propertyValues {Curve {keys [{interpMode=RCIM_Cubic, tangentMode=RCTM_User, time=0, value=0}, {interpMode=RCIM_Cubic, tangentMode=RCTM_User, time=1, value=1}]}}})
    )
    Index 2 Param 1=&FloatFromCurve(FloatCurve=DataInterface() {propertyValues {Curve {keys [
         {arriveTangent=1.3287, interpMode=RCIM_Cubic, leaveTangent=1.3287, tangentMode=RCTM_User, time=0, value=0}
         {interpMode=RCIM_Cubic, leaveTangent=1.3759, time=1, value=1}
        ]}}})
    Index 2 Param 2=&RandomRangeFloat(Minimum=-0.25, Maximum=0.25)
    Index 2 Param 3=&Lerp_Float(
     A=1
     B=1
     Alpha=&FloatFromCurve(FloatCurve=DataInterface() {propertyValues {Curve {keys [{tangentMode=RCTM_User, time=0, value=0}, {tangentMode=RCTM_User, time=1, value=1}]}}})
    )
    Index 2 Param 4=&RandomRangeFloat(Minimum=0.45, Maximum=0.5)
    Module.Index0Param0Write=true
    Module.Index0Param1Write=true
    Module.Index0Param2Write=true
    Module.Index0Param3Write=true
    Module.Index2Param0Write=true
    Module.Index2Param1Write=true
    Module.Index2Param2Write=true
    Module.Index2Param3Write=true
    Write Parameter Index 2=true
   )
  ]
  rendererClasses [&NiagaraSpriteRendererProperties]
  renderers [
   {rendererClass=&NiagaraSpriteRendererProperties}
   {rendererClass=&NiagaraSpriteRendererProperties, rendererIndex=1}
  ]
 }
 Splash {
  bEnabled=false
  emitterUpdateScript [
   &EmitterState()
   &Spawn_NDCFortImpacts(Max Spawn Count=1, Min Spawn Count=1, Reader:NiagaraDataInterfaceDataChannelRead=$System.NDCReader)
  ]
  particleSpawnScript [
   &InitializeParticleFromNDC_FortWaterInteraction(FallbackScaleMultiplier=50)
   &HSV_Offset(Input Color:LinearColor=&MakeCustomLinearColorFromBool(Bool=$System.DoDebugColor, True Color=$System.DebugWaterColor, False Color=$Output.InitializeParticleFromNDC_FortWaterInteraction.WaterColor))
   /SetVariables(Particles.AmmoTypeScale:NiagaraFloat=&Multiply_Float(
     A=&SelectFloatFromArray(
      Direct Array Index=$Output.InitializeParticleFromNDC_FortWaterInteraction.AmmoType
      Float Selection Array=DataInterface() {propertyValues {FloatData [0.8,1,1.5,0.75,1]}}
     )
     B=$Output.InitializeParticleFromNDC_FortWaterInteraction.ImpactScaleMultiplier
    )) {isSetParameters=true}
   &InitializeParticle(
    Color=&MakeLinearColor(
     R:NiagaraFloat=&MakeFloatFromLinearColor(LinearColor=$Particles.HSVOffset, Channel:ENiagaraFloatFromLinearColorOptions=::NewEnumerator0)
     G:NiagaraFloat=&MakeFloatFromLinearColor(LinearColor=$Particles.HSVOffset, Channel:ENiagaraFloatFromLinearColorOptions=::NewEnumerator1)
     B:NiagaraFloat=&MakeFloatFromLinearColor(LinearColor=$Particles.HSVOffset, Channel:ENiagaraFloatFromLinearColorOptions=::NewEnumerator2)
     A:NiagaraFloat=1
    )
    Lifetime Max=0.5
    Lifetime Min=0.4
    Lifetime Mode=::NewEnumerator1
    Position=$Output.InitializeParticleFromNDC_FortWaterInteraction.ImpactPoint
    Position Mode=::NewEnumerator1
    Position Offset=<0 0 2>
    Sprite Rotation Mode=::NewEnumerator2
    Sprite Size=&Multiply_Vector2D_ByFloat(
     Float=&Multiply_Float(A=&RandomRangeFloat(Minimum=0.8, Maximum=1), B=$Particles.AmmoTypeScale)
     Vector2D=<110 110>
    )
    Sprite Size Mode=::NewEnumerator4
    UsePositionOffset=true
    Write Material Random=true
   )
   &AddVelocityInCone(
    Cone Angle=3
    Cone Axis=$Output.InitializeParticleFromNDC_FortWaterInteraction.ImpactNormal
    Module.Use Velocity Falloff On Cone Axis=false
    Velocity Strength=&Multiply_Float(A=&UniformRangedFloat(Minimum=200, Maximum=225), B=$Particles.AmmoTypeScale)
   )
   /Niagara/Modules/Update/SubUV/V2/SubUVAnimation(
    Evaluation Type=::NewEnumerator0
    "Sprite Renderer "=DataInterface() {propertyValues {SpriteRenderer=&self:Splash_4.NiagaraSpriteRendererProperties_0}}
    SubUV Animation Mode:ENiagara_SubUVLookupModeV2=::NewEnumerator2
    UseEndFrame=false
    UseStartFrame=false
   ) {module=SubUVAnimation001}
  ]
  particleUpdateScript [
   &ParticleState()
   &GravityForce(Gravity=&Multiply_VectorByFloat(Vector=<0 0 -980>, Float=1))
   &SolveForcesAndVelocity(Clamp Velocity=false, Limit Acceleration=false)
   &ScaleSpriteSize(Scale Factor=&Vector2DFromFloat(Value=&FloatFromCurve(FloatCurve=DataInterface() {propertyValues {Curve {keys [<at 0 0.2>,<at 1 1>]}}})))
   &DynamicMaterialParameters(
    Index 0 Param 2=&FloatFromCurve(FloatCurve=DataInterface() {propertyValues {Curve {keys [{interpMode=RCIM_Cubic, tangentMode=RCTM_User, time=0, value=1}, {interpMode=RCIM_Cubic, tangentMode=RCTM_User, time=1, value=0}]}}})
    Index 0 Param 4=&FloatFromCurve(FloatCurve=DataInterface() {propertyValues {Curve {keys [{tangentMode=RCTM_User, time=0, value=0}, {tangentMode=RCTM_User, time=1, value=1}]}}})
    Module.Param0WriteEnabled=false
    Module.Param1WriteEnabled=true
    Module.Param2WriteEnabled=false
    Module.Param3WriteEnabled=true
   )
   &Color(Color=&MakeLinearColorFromVectorAndFloat(
     "Vector (RGB)"=&MakeVectorFromLinearColorRGB(Color=$Particles.Color)
     "Float (Alpha)"=&FloatFromCurve(FloatCurve=DataInterface() {propertyValues {Curve {keys [<at 0.4 1 RCIM_Cubic>,<at 1 0>]}}})
    ))
  ]
  rendererClasses [&NiagaraSpriteRendererProperties]
  renderers [{rendererClass=&NiagaraSpriteRendererProperties}]
 }
 Splash001 {
  emitterUpdateScript [
   &EmitterState()
   &Spawn_NDCFortImpacts(Max Spawn Count=1, Min Spawn Count=1, Reader:NiagaraDataInterfaceDataChannelRead=$System.NDCReader)
  ]
  particleSpawnScript [
   &InitializeParticleFromNDC_FortWaterInteraction(FallbackScaleMultiplier=50)
   &HSV_Offset(Input Color:LinearColor=&MakeCustomLinearColorFromBool(Bool=$System.DoDebugColor, True Color=$System.DebugWaterColor, False Color=$Output.InitializeParticleFromNDC_FortWaterInteraction.WaterColor))
   /SetVariables(Particles.AmmoTypeScale:NiagaraFloat=&Multiply_Float(
     A=&SelectFloatFromArray(
      Direct Array Index=$Output.InitializeParticleFromNDC_FortWaterInteraction.AmmoType
      Float Selection Array=DataInterface() {propertyValues {FloatData [0.8,1,1.5,0.75,1]}}
     )
     B=$Output.InitializeParticleFromNDC_FortWaterInteraction.ImpactScaleMultiplier
    )) {isSetParameters=true}
   &InitializeParticle(
    Color=&MakeLinearColor(
     R:NiagaraFloat=&MakeFloatFromLinearColor(LinearColor=$Particles.HSVOffset, Channel:ENiagaraFloatFromLinearColorOptions=::NewEnumerator0)
     G:NiagaraFloat=&MakeFloatFromLinearColor(LinearColor=$Particles.HSVOffset, Channel:ENiagaraFloatFromLinearColorOptions=::NewEnumerator1)
     B:NiagaraFloat=&MakeFloatFromLinearColor(LinearColor=$Particles.HSVOffset, Channel:ENiagaraFloatFromLinearColorOptions=::NewEnumerator2)
     A:NiagaraFloat=1
    )
    Lifetime Max=0.4
    Lifetime Min=0.3
    Lifetime Mode=::NewEnumerator1
    Position=$Output.InitializeParticleFromNDC_FortWaterInteraction.ImpactPoint
    Position Mode=::NewEnumerator1
    Position Offset=<0 0 2>
    Sprite Size=&Multiply_Vector2D_ByFloat(
     Float=&Multiply_Float(A=&RandomRangeFloat(Minimum=0.8, Maximum=1), B=$Particles.AmmoTypeScale)
     Vector2D=<30 80>
    )
    Sprite Size Mode=::NewEnumerator4
    UsePositionOffset=true
    Write Material Random=true
   )
   &AddVelocityInCone(
    Cone Angle=5
    Cone Axis=$Output.InitializeParticleFromNDC_FortWaterInteraction.ImpactNormal
    Module.Use Velocity Falloff On Cone Axis=false
    Velocity Strength=&Multiply_Float(A=&UniformRangedFloat(Minimum=50, Maximum=100), B=$Particles.AmmoTypeScale)
   )
   &SubUVAnimation(
    Evaluation Type=::NewEnumerator0
    "Sprite Renderer "=DataInterface() {propertyValues {SpriteRenderer=&self:Splash001_6.NiagaraSpriteRendererProperties_0}}
    SubUV Animation Mode:ENiagara_SubUVLookupModeV2=::NewEnumerator2
    UseEndFrame=false
    UseStartFrame=false
   ) {module=SubUVAnimation001}
   &SpriteFacingAndAlignment(Sprite Alignment=$Particles.Velocity, WriteAlignment=true, WriteFacing=false)
  ]
  particleUpdateScript [
   &ParticleState()
   &GravityForce(Gravity=&Multiply_VectorByFloat(Vector=<0 0 -980>, Float=1))
   &SolveForcesAndVelocity(Clamp Velocity=false, Limit Acceleration=false)
   &ScaleSpriteSize(Uniform Curve Sprite Scale=DataInterface() {propertyValues {Curve {keys [{arriveTangent=2.2272, interpMode=RCIM_Cubic, leaveTangent=2.2272, tangentMode=RCTM_User, time=0, value=0.33}, <at 0.5 1 RCIM_Cubic>]}}})
   &DynamicMaterialParameters(
    Index 0 Param 2=&FloatFromCurve(FloatCurve=DataInterface() {propertyValues {Curve {keys [{interpMode=RCIM_Cubic, tangentMode=RCTM_User, time=0, value=1}, {interpMode=RCIM_Cubic, tangentMode=RCTM_User, time=1, value=0}]}}})
    Index 0 Param 4=&FloatFromCurve(FloatCurve=DataInterface() {propertyValues {Curve {keys [{tangentMode=RCTM_User, time=0, value=0}, {tangentMode=RCTM_User, time=1, value=1}]}}})
    Module.Param0WriteEnabled=false
    Module.Param1WriteEnabled=true
    Module.Param2WriteEnabled=false
    Module.Param3WriteEnabled=true
   )
   &Color(Color=&MakeLinearColorFromVectorAndFloat(
     "Vector (RGB)"=&MakeVectorFromLinearColorRGB(Color=$Particles.Color)
     "Float (Alpha)"=&FloatFromCurve(FloatCurve=DataInterface() {propertyValues {Curve {keys [<at 0.5 1 RCIM_Cubic>,<at 1 0>]}}})
    ))
  ]
  rendererClasses [&NiagaraSpriteRendererProperties]
  renderers [{rendererClass=&NiagaraSpriteRendererProperties}]
 }
}
summary {
 systemName=NS_NDCHandler_WaterInteraction
 emitters [
  {bEnabled=false, emitterName=Impact_Surface_Foam, rendererClasses [&NiagaraMeshRendererProperties]}
  {bEnabled=false, emitterName=Drops_Wide001, rendererClasses [&NiagaraSpriteRendererProperties]}
  {bEnabled=false, emitterName=Drops_Tall, rendererClasses [&NiagaraSpriteRendererProperties]}
  {bEnabled=false, emitterName=Splash, rendererClasses [&NiagaraSpriteRendererProperties]}
  {emitterName=Splash001, rendererClasses [&NiagaraSpriteRendererProperties]}
  {bEnabled=false, emitterName=Mist, rendererClasses [&NiagaraSpriteRendererProperties]}
  {emitterName=Drops_Tall_Splashy, rendererClasses [&NiagaraSpriteRendererProperties]}
  {emitterName=Drops_Wide_Splashy, simTarget=GPUComputeSim, rendererClasses [&NiagaraSpriteRendererProperties]}
  {emitterName=FN_Weapons_Debris_Sprite_SubUV_BETA, rendererClasses [&NiagaraSpriteRendererProperties]}
  {emitterName=Impact_Surface_Foam001, rendererClasses [&NiagaraMeshRendererProperties]}
 ]
 userVariables []
}
